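# Fortnite credits screen widget that auto-scrolls logos, staff credits and legal notices, showing the BattlEye notice only when in use.
# On construct, it builds one tagged rich-text string from CreditsDataTable, styled by entry type.
# Activation resets scrolling and fades out a black mask, then a delayed looping timer advances the scroll offset.
# Holding the mouse speeds scrolling; reaching the captured content height restarts the sequence, and deactivation stops the timer.
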WidgetBlueprint /FortniteCoreUI/Frontend/Credits/Credits
animations [FadeMask-Show {
  binding_count=3
  display_label=""
  length_seconds=2.5
  object=&self:FadeMask-Show
  sequence {bindings [
    Image_1 {id=#1, possessed_class=Image, tracks []}
    ImageMask {
     id=#2
     possessed_class=Image
     tracks [{
       type=MovieSceneColorTrack
       display_name=Color and Opacity
       sections [{
         type=MovieSceneColorSection
         end_frame=50
         start_frame=0
         channels [
          Color.R {type=MovieSceneScriptingFloatChannel, const [0,1]}
          Color.G {type=MovieSceneScriptingFloatChannel, const [0,1]}
          Color.B {type=MovieSceneScriptingFloatChannel, const [0,1]}
          Color.A {type=MovieSceneScriptingFloatChannel, keys [[0,1],[10,1],[50,0]]}
         ]
        }]
      }]
    }
    ScaleBoxSlot_1 {
     id=#3
     parent_id=#1
     possessed_class=ScaleBoxSlot
     tracks [{
       type=MovieSceneMarginTrack
       display_name=Padding
       sections [{
         type=MovieSceneMarginSection
         end_frame=50
         start_frame=50
         channels [
          Left {type=MovieSceneScriptingFloatChannel}
          Top {type=MovieSceneScriptingFloatChannel}
          Right {type=MovieSceneScriptingFloatChannel}
          Bottom {type=MovieSceneScriptingFloatChannel, const [0,1]}
         ]
        }]
      }]
    }
   ]}
 }]
graphs [
 CaptureEndPoint {nodes [
   CaptureEndPoint() {is=title K2Node_FunctionEntry}
   n1 = Get ScrollBoxCredits() -> (ScrollBoxCredits:Scroll Box Object Reference)
   > Set EndPointOffset(EndPointOffset:double=BreakVector2D(InVec=GetDesiredSize!(self:Widget Object Reference=@n1) -> (ReturnValue:Vector 2D Structure)).Y) -> (Output_Get:double)
  ]}
 EventGraph {nodes [
   Delay(Duration=2)
   n1 = Get ScrollBoxCredits()
   > ScrollToEnd(self:Scroll Box Object Reference=@n1)
   > Delay(Duration=2)
   > PrintString!(
    execute
    InString="To String (Float)"(InDouble=GetScrollOffset!(self:Scroll Box Object Reference=@n1) -> (ReturnValue:float))
    bPrintToScreen
    bPrintToLog
    TextColor=(R=0.9561,G=1,B=0.9534,A=1)
    Duration
    Key
   ) -> (then)
   n2 = HandleFadeInFinished() {is=title K2Node_CustomEvent}
   > Delay(Duration=3)
   > CaptureEndPoint() {is=title K2Node_CallFunction}
   > ToggleScrollTimer!(execute, self, EnableTimer:bool=true) -> (then)
   Event Construct()
   > Populate Credits Text()
   > BindToAnimationFinished(Animation=Get FadeMask-Show() -> (FadeMask-Show:Widget Animation Object Reference), Delegate=@n2)
   Event On Deactivated()
   > ToggleScrollTimer(EnableTimer=false)
   Event On Activated()
   > SetVisibility(
    self=Get Vboxbattleeye() -> (Vboxbattleeye:Vertical Box Object Reference)
    InVisibility=Select(Option 0=Collapsed, Option 1=SelfHitTestInvisible, Index=IsUsingBattlEye!() -> (ReturnValue:bool))
   )
   > Reset()
  ]}
 OnMouseButtonDown {nodes [
   OnMouseButtonDown!() -> (then, MyGeometry:Geometry Structure, MouseEvent:"Pointer Event Structure (by ref)") {is=title K2Node_FunctionEntry}
   > Set ScrollSpeed(ScrollSpeed:double=10) -> (Output_Get:double)
   > Return Node!(execute, ReturnValue:Event Reply Structure=Handled())
  ]}
 OnMouseButtonUp {nodes [
   OnMouseButtonUp!() -> (then, MyGeometry:Geometry Structure, MouseEvent:"Pointer Event Structure (by ref)") {is=title K2Node_FunctionEntry}
   > Set ScrollSpeed(ScrollSpeed=1)
   > Return Node(ReturnValue=Handled())
  ]}
 Populate Credits Text {nodes [
   Populate Credits Text() {is=title K2Node_FunctionEntry}
   > Set TempString(TempString:String) -> (Output_Get:String)
   > SetText(
    self:Fort Rich Text Block Legacy Object Reference=Get FortRichTextBlock_0() -> (FortRichTextBlock_0:Fort Rich Text Block Legacy Object Reference)
    InText:Text=GetEmptyText!() -> (ReturnValue:Text)
   )
   > n1 = GetDataTableRowNames(Table=&/CreditsDataTable)
   > n2 = For Each Loop(Array:Array of Names=@n1) -> (Array Element:Name)
   > n3 = Get Data Table Row CreditsDataTable!(execute:Exec, DataTable:Data Table Object Reference=&CreditsDataTable, RowName:Name=@n2) -> (then:Exec, RowNotFound:Exec, ReturnValue:Credits Entry Struct Structure) {is=title K2Node_GetDataTableRow}
   n4 = Break Credits Entry Struct!(CreditsEntryStruct:"Credits Entry Struct Structure (by ref)"=@n3) -> (EntryDisplayText_10_#4:Text, EntryType_5_#5:CreditsEntryType Enum) {is=title K2Node_BreakStruct}
   > n5 = Switch on CreditsEntryType!(execute, Selection:CreditsEntryType Enum=@n4.EntryType_5_#5) -> (NewEnumerator0:Exec, NewEnumerator1:Exec, NewEnumerator2:Exec, NewEnumerator3:Exec) {is=title K2Node_SwitchEnum}
   n6 = "To String (Text)"(InText=@n4)
   > n7 = Set TempString(TempString=Append(A=Get TempString() -> (TempString:String), B=@n6))
   Set TempString(execute=[@n10.then, @n9.then, @n8.then, @n7.then], TempString=Append(A=Get TempString(), B="\r\n"))
   n8 = Set TempString(execute=@n5.NewEnumerator2, TempString=Append(A=Get TempString(), B=Append!(A="<Subheader>", B=@n6, C:String="</>") -> (ReturnValue)))
   n9 = Set TempString(
    execute=@n5.NewEnumerator1
    TempString=Append!(A=Get TempString(), B=Append!(A="<Header>", B=@n6, C:String="</>") -> (ReturnValue)) -> (ReturnValue)
   )
   n10 = Set TempString(execute=@n5.NewEnumerator3, TempString=Append!(A=Get TempString(), B) -> (ReturnValue))
   SetText(execute=@n2.Completed, self=Get FortRichTextBlock_0(), InText="To Text (String)"(InString=Get TempString()))
   > Return Node!(execute)
  ]}
 Reset {nodes [
   Reset() {is=title K2Node_FunctionEntry}
   > SetScrollOffset!(execute, self:Scroll Box Object Reference=Get ScrollBoxCredits(), NewScrollOffset:float=0) -> (then)
   > PlayAnimation(InAnimation=Get FadeMask-Show())
  ]}
 ScrollCredits {nodes [
   ScrollCredits() {is=title K2Node_FunctionEntry}
   n1 = GetScrollOffset(self=Get ScrollBoxCredits())
   > Set CurrentPosition(CurrentPosition:double=@n1) -> (Output_Get:double)
   > SetScrollOffset(
    self=Get ScrollBoxCredits()
    NewScrollOffset=float + float(A=Get CurrentPosition() -> (CurrentPosition:double), B=Get ScrollSpeed() -> (ScrollSpeed:double)) {is=title K2Node_CommutativeAssociativeBinaryOperator}
   )
   > Branch(Condition="float >= float"(A=@n1, B=Get EndPointOffset() -> (EndPointOffset:double)) {is=title K2Node_CallFunction})
   > ToggleScrollTimer(EnableTimer=false)
   > Reset() {is=title K2Node_CallFunction}
  ]}
 ToggleScrollTimer {nodes [
   n1 = ToggleScrollTimer!() -> (then, EnableTimer:bool=~) {is=title K2Node_FunctionEntry}
   > n2 = Branch(Condition=@n1)
   > n3 = Set Timer by Event(Delegate=Create Event(self=Self-Reference()), Time=.01, bLooping=true)
   > Set ScrollTimerHandle(ScrollTimerHandle:Timer Handle Structure=@n3) -> (Output_Get:Timer Handle Structure)
   Clear and Invalidate Timer by Handle(execute=@n2.else, Handle=Get ScrollTimerHandle() -> (ScrollTimerHandle:Timer Handle Structure))
  ]}
]
variables [
 EndPointOffset {type=number}
 ScrollSpeed {type=number, default=1}
 ScrollTimerHandle {type=TimerHandle}
 TempString {type=string, default=""}
]
widgets {
 info {inheritedWidgetCount=1, parentClass=/Script/FortniteUI.FortCreditsScreen, rootWidgetClass=&Overlay, widgetCount=39}
 widgets [
  Overlay_1 {widgetClassPath=&Overlay, widget=Overlay_1}
  BorderShell {
   widgetClassPath=&CommonBorder
   parent=Overlay_1
   widget=BorderShell
   properties {
    background {
     bIsDynamicallyLoaded=false
     drawAs=Image
     imageType=NoImage
     mirroring=NoMirror
     resourceName=None
     resourceObject=null
     tiling=NoTile
     imageSize=<32 32>
     margin=<ltrb 0 0 0 0>
     outlineSettings {
      bUseBrushTransparency=false
      roundingType=HalfHeightRadius
      width=0
      color {colorUseRule=UseColor_Specified, specifiedColor=<rgba 0 0 0 0>}
      cornerRadii=<0 0 0 0>
     }
     tintColor {colorUseRule=UseColor_Specified, specifiedColor=<rgba 0 0 0 1>}
     uVRegion {bIsValid=false, max=<0 0>, min=<0 0>}
    }
    padding=<ltrb 0 0 0 0>
   }
   slot {class=OverlaySlot, properties {horizontalAlignment=HAlign_Fill, verticalAlignment=VAlign_Fill}}
  }
  SafeZone_2 {
   widgetClassPath=&SafeZone
   bIsVariable=true
   parent=BorderShell
   widget=SafeZone_2
   properties {padBottom=false}
   slot {class=BorderSlot, properties {padding=<ltrb 0 0 0 0>}}
  }
  ScrollBoxCredits {
   widgetClassPath=&ScrollBox
   bIsVariable=true
   parent=SafeZone_2
   widget=ScrollBoxCredits
   properties {consumeMouseWheel=Never, scrollBarVisibility=Collapsed}
   slot {class=SafeZoneSlot}
  }
  Spacer_1 {widgetClassPath=&Spacer, parent=ScrollBoxCredits, widget=Spacer_1, properties {size=<1 224>}, slot {class=ScrollBoxSlot}}
  SizeBox_1 {
   widgetClassPath=&SizeBox
   parent=ScrollBoxCredits
   widget=SizeBox_1
   properties {bOverride_WidthOverride=true, heightOverride=350, widthOverride=700}
   slot {class=ScrollBoxSlot, properties {horizontalAlignment=HAlign_Center, padding=<ltrb 0 -150 0 -100>}}
  }
  ScaleBox_1 {widgetClassPath=&ScaleBox, parent=SizeBox_1, widget=ScaleBox_1, slot {class=SizeBoxSlot}}
  Image_1 {
   widgetClassPath=&Image
   parent=ScaleBox_1
   widget=Image_1
   properties {brush {
     bIsDynamicallyLoaded=false
     drawAs=Image
     imageType=NoImage
     mirroring=NoMirror
     resourceName=None
     resourceObject=/FortniteCoreUI/Foundation/Textures/Logo/T-Logo-Wordmark-1024x512
     tiling=NoTile
     imageSize=<1024 512>
     margin=<ltrb 0 0 0 0>
     outlineSettings {
      bUseBrushTransparency=false
      roundingType=HalfHeightRadius
      width=0
      color {colorUseRule=UseColor_Specified, specifiedColor=<rgba 0 0 0 0>}
      cornerRadii=<0 0 0 0>
     }
     tintColor {colorUseRule=UseColor_Specified, specifiedColor=<rgba 1 1 1 1>}
     uVRegion {bIsValid=false, max=<0 0>, min=<0 0>}
    }}
   slot {class=ScaleBoxSlot}
  }
  Spacer_4 {widgetClassPath=&Spacer, parent=ScrollBoxCredits, widget=Spacer_4, properties {size=<1 128>}, slot {class=ScrollBoxSlot}}
  TextDedication {
   widgetClassPath=&CommonTextBlock
   parent=ScrollBoxCredits
   widget=TextDedication
   properties {
    justification=Center
    style=/FortUILibrary/Styles/Typography/Foundation/TextStyle-Base-M-B.TextStyle-Base-M-B_C
    text=For
    font {
     bForceMonospaced=false
     bMaterialIsStencil=false
     fontMaterial=null
     fontObject=&NotoSans
     letterSpacing=0
     monospacedWidth=1
     size=24
     skewAmount=0
     typefaceFontName=NotoSans-Bold
     outlineSettings {
      bApplyOutlineToDropShadows=false
      bMiteredCorners=false
      bSeparateFillAlpha=false
      outlineMaterial=null
      outlineSize=0
      outlineColor=<rgba 0 0 0 1>
     }
    }
    shadowOffset=<0 0>
   }
   slot {class=ScrollBoxSlot, properties {horizontalAlignment=HAlign_Center, padding=<ltrb 0 0 0 18>}}
  }
  CommonTextBlock_0 {
   widgetClassPath=&CommonTextBlock
   parent=ScrollBoxCredits
   widget=CommonTextBlock_0
   properties {
    justification=Center
    style=&TextStyle-Base-M-B_C
    text=Karolina
    font {
     bForceMonospaced=false
     bMaterialIsStencil=false
     fontMaterial=null
     fontObject=&NotoSans
     letterSpacing=0
     monospacedWidth=1
     size=24
     skewAmount=0
     typefaceFontName=NotoSans-Bold
     outlineSettings {
      bApplyOutlineToDropShadows=false
      bMiteredCorners=false
      bSeparateFillAlpha=false
      outlineMaterial=null
      outlineSize=0
      outlineColor=<rgba 0 0 0 1>
     }
    }
    shadowOffset=<0 0>
   }
   slot {class=ScrollBoxSlot, properties {horizontalAlignment=HAlign_Center, padding=<ltrb 0 0 0 18>}}
  }
  CommonTextBlock_1 {
   widgetClassPath=&CommonTextBlock
   parent=ScrollBoxCredits
   widget=CommonTextBlock_1
   properties {
    justification=Center
    style=&TextStyle-Base-M-B_C
    text="&&"
    font {
     bForceMonospaced=false
     bMaterialIsStencil=false
     fontMaterial=null
     fontObject=&NotoSans
     letterSpacing=0
     monospacedWidth=1
     size=24
     skewAmount=0
     typefaceFontName=NotoSans-Bold
     outlineSettings {
      bApplyOutlineToDropShadows=false
      bMiteredCorners=false
      bSeparateFillAlpha=false
      outlineMaterial=null
      outlineSize=0
      outlineColor=<rgba 0 0 0 1>
     }
    }
    shadowOffset=<0 0>
   }
   slot {class=ScrollBoxSlot, properties {horizontalAlignment=HAlign_Center, padding=<ltrb 0 0 0 18>}}
  }
  CommonTextBlock_2 {
   widgetClassPath=&CommonTextBlock
   parent=ScrollBoxCredits
   widget=CommonTextBlock_2
   properties {
    justification=Center
    style=&TextStyle-Base-M-B_C
    text=Joel
    font {
     bForceMonospaced=false
     bMaterialIsStencil=false
     fontMaterial=null
     fontObject=&NotoSans
     letterSpacing=0
     monospacedWidth=1
     size=24
     skewAmount=0
     typefaceFontName=NotoSans-Bold
     outlineSettings {
      bApplyOutlineToDropShadows=false
      bMiteredCorners=false
      bSeparateFillAlpha=false
      outlineMaterial=null
      outlineSize=0
      outlineColor=<rgba 0 0 0 1>
     }
    }
    shadowOffset=<0 0>
   }
   slot {class=ScrollBoxSlot, properties {horizontalAlignment=HAlign_Center}}
  }
  Spacer_3 {widgetClassPath=&Spacer, parent=ScrollBoxCredits, widget=Spacer_3, properties {size=<1 320>}, slot {class=ScrollBoxSlot}}
  TextCompany {
   widgetClassPath=&CommonTextBlock
   parent=ScrollBoxCredits
   widget=TextCompany
   properties {
    justification=Center
    style=/FortUILibrary/Styles/Typography/Foundation/TextStyle-Base-S-B.TextStyle-Base-S-B_C
    text="Designed, Developed, and Published by"
    font {
     bForceMonospaced=false
     bMaterialIsStencil=false
     fontMaterial=null
     fontObject=&NotoSans
     letterSpacing=0
     monospacedWidth=1
     size=18
     skewAmount=0
     typefaceFontName=NotoSans-Bold
     outlineSettings {
      bApplyOutlineToDropShadows=false
      bMiteredCorners=false
      bSeparateFillAlpha=false
      outlineMaterial=null
      outlineSize=0
      outlineColor=<rgba 0 0 0 1>
     }
    }
    shadowOffset=<0 0>
   }
   slot {class=ScrollBoxSlot, properties {horizontalAlignment=HAlign_Center}}
  }
  Spacer_0 {widgetClassPath=&Spacer, parent=ScrollBoxCredits, widget=Spacer_0, properties {size=<1 64>}, slot {class=ScrollBoxSlot}}
  SizeBox_0 {
   widgetClassPath=&SizeBox
   parent=ScrollBoxCredits
   widget=SizeBox_0
   properties {bOverride_HeightOverride=true, heightOverride=256, widthOverride=145.9556}
   slot {class=ScrollBoxSlot}
  }
  ScaleBox_0 {
   widgetClassPath=&ScaleBox
   parent=SizeBox_0
   widget=ScaleBox_0
   slot {class=SizeBoxSlot, properties {horizontalAlignment=HAlign_Center}}
  }
  Image_0 {
   widgetClassPath=&Image
   parent=ScaleBox_0
   widget=Image_0
   properties {brush {
     bIsDynamicallyLoaded=false
     drawAs=Image
     imageType=NoImage
     mirroring=NoMirror
     resourceName=None
     resourceObject=/FortniteCoreUI/UI/Foundation/Textures/Banner/Event_Special/T-Banners-Icons-Epic-L
     tiling=NoTile
     imageSize=<256 256>
     margin=<ltrb 0 0 0 0>
     outlineSettings {
      bUseBrushTransparency=false
      roundingType=HalfHeightRadius
      width=0
      color {colorUseRule=UseColor_Specified, specifiedColor=<rgba 0 0 0 0>}
      cornerRadii=<0 0 0 0>
     }
     tintColor {colorUseRule=UseColor_Specified, specifiedColor=<rgba 1 1 1 1>}
     uVRegion {bIsValid=false, max=<0 0>, min=<0 0>}
    }}
   slot {class=ScaleBoxSlot, properties {horizontalAlignment=HAlign_Fill, verticalAlignment=VAlign_Fill}}
  }
  Spacer_2 {widgetClassPath=&Spacer, parent=ScrollBoxCredits, widget=Spacer_2, properties {size=<1 64>}, slot {class=ScrollBoxSlot}}
  FortRichTextBlock_0 {
   widgetClassPath=/Script/FortniteUI.FortRichTextBlockLegacy
   bIsVariable=true
   parent=ScrollBoxCredits
   widget=FortRichTextBlock_0
   properties {autoWrapText=true, justification=Center, styleSet=/FortniteCoreUI/UI/Foundation/Legal/CreditsScreenRichTextStyles}
   slot {class=ScrollBoxSlot, properties {horizontalAlignment=HAlign_Center}}
  }
  TextThanks {
   widgetClassPath=&CommonTextBlock
   parent=ScrollBoxCredits
   widget=TextThanks
   properties {
    justification=Center
    style=/FortUILibrary/Styles/Typography/Foundation/TextStyle-Base-XS.TextStyle-Base-XS_C
    text="Thanks to all the UX test participants and alpha testers who tested the game or answered our surveys; you all contributed to improving Fortnite!"
    wrapTextAt=941.4548
    font {
     bForceMonospaced=false
     bMaterialIsStencil=false
     fontMaterial=null
     fontObject=&NotoSans
     letterSpacing=0
     monospacedWidth=1
     size=16
     skewAmount=0
     typefaceFontName=NotoSans-Regular
     outlineSettings {
      bApplyOutlineToDropShadows=false
      bMiteredCorners=false
      bSeparateFillAlpha=false
      outlineMaterial=null
      outlineSize=0
      outlineColor=<rgba 0 0 0 1>
     }
    }
    shadowOffset=<0 0>
   }
   slot {class=ScrollBoxSlot, properties {horizontalAlignment=HAlign_Center}}
  }
  Spacer_7 {widgetClassPath=&Spacer, parent=ScrollBoxCredits, widget=Spacer_7, properties {size=<1 200>}, slot {class=ScrollBoxSlot}}
  Vboxbattleeye {
   widgetClassPath=&VerticalBox
   bIsVariable=true
   parent=ScrollBoxCredits
   widget=Vboxbattleeye
   slot {class=ScrollBoxSlot, properties {horizontalAlignment=HAlign_Center}}
  }
  SizeBox_2 {
   widgetClassPath=&SizeBox
   parent=Vboxbattleeye
   widget=SizeBox_2
   properties {bOverride_HeightOverride=true, bOverride_WidthOverride=true, heightOverride=256, widthOverride=256}
   slot {class=VerticalBoxSlot, properties {horizontalAlignment=HAlign_Center}}
  }
  Image_2 {
   widgetClassPath=&Image
   parent=SizeBox_2
   widget=Image_2
   properties {brush {
     bIsDynamicallyLoaded=false
     drawAs=Image
     imageType=NoImage
     mirroring=NoMirror
     resourceName=None
     resourceObject=/FortniteCoreUI/Foundation/Textures/BattleEyeLogo/T-BattleEyeLogo
     tiling=NoTile
     imageSize=<512 512>
     margin=<ltrb 0 0 0 0>
     outlineSettings {
      bUseBrushTransparency=false
      roundingType=HalfHeightRadius
      width=0
      color {colorUseRule=UseColor_Specified, specifiedColor=<rgba 0 0 0 0>}
      cornerRadii=<0 0 0 0>
     }
     tintColor {colorUseRule=UseColor_Specified, specifiedColor=<rgba 1 1 1 1>}
     uVRegion {bIsValid=false, max=<0 0>, min=<0 0>}
    }}
   slot {class=SizeBoxSlot}
  }
  SizeBox_3 {
   widgetClassPath=&SizeBox
   parent=Vboxbattleeye
   widget=SizeBox_3
   properties {bOverride_WidthOverride=true, widthOverride=1200}
   slot {class=VerticalBoxSlot, properties {horizontalAlignment=HAlign_Center, padding=<ltrb 0 32 0 0>}}
  }
  CommonTextBlock_3 {
   widgetClassPath=&CommonTextBlock
   parent=SizeBox_3
   widget=CommonTextBlock_3
   properties {
    autoWrapText=true
    justification=Center
    style=&TextStyle-Base-XS_C
    text="This software product includes BattlEye™ anti-cheat technology, © 2017 BattlEye Innovations e.K.. All rights reserved."
    font {
     bForceMonospaced=false
     bMaterialIsStencil=false
     fontMaterial=null
     fontObject=&NotoSans
     letterSpacing=0
     monospacedWidth=1
     size=16
     skewAmount=0
     typefaceFontName=NotoSans-Regular
     outlineSettings {
      bApplyOutlineToDropShadows=false
      bMiteredCorners=false
      bSeparateFillAlpha=false
      outlineMaterial=null
      outlineSize=0
      outlineColor=<rgba 0 0 0 1>
     }
    }
    shadowOffset=<0 0>
   }
   slot {class=SizeBoxSlot}
  }
  Spacer_8 {widgetClassPath=&Spacer, parent=ScrollBoxCredits, widget=Spacer_8, properties {size=<1 200>}, slot {class=ScrollBoxSlot}}
  VboxVivox {
   widgetClassPath=&VerticalBox
   parent=ScrollBoxCredits
   widget=VboxVivox
   slot {class=ScrollBoxSlot, properties {horizontalAlignment=HAlign_Center}}
  }
  SizeBox_4 {
   widgetClassPath=&SizeBox
   parent=VboxVivox
   widget=SizeBox_4
   properties {bOverride_HeightOverride=true, bOverride_WidthOverride=true, heightOverride=256, widthOverride=256}
   slot {class=VerticalBoxSlot, properties {horizontalAlignment=HAlign_Center}}
  }
  Image_3 {
   widgetClassPath=&Image
   parent=SizeBox_4
   widget=Image_3
   properties {brush {
     bIsDynamicallyLoaded=false
     drawAs=Image
     imageType=NoImage
     mirroring=NoMirror
     resourceName=None
     resourceObject=/FortniteCoreUI/Foundation/Textures/VivoxLogo/T-Vivox_Logo
     tiling=NoTile
     imageSize=<512 512>
     margin=<ltrb 0 0 0 0>
     outlineSettings {
      bUseBrushTransparency=false
      roundingType=HalfHeightRadius
      width=0
      color {colorUseRule=UseColor_Specified, specifiedColor=<rgba 0 0 0 0>}
      cornerRadii=<0 0 0 0>
     }
     tintColor {colorUseRule=UseColor_Specified, specifiedColor=<rgba 1 1 1 1>}
     uVRegion {bIsValid=false, max=<0 0>, min=<0 0>}
    }}
   slot {class=SizeBoxSlot}
  }
  Spacer_5 {widgetClassPath=&Spacer, parent=ScrollBoxCredits, widget=Spacer_5, properties {size=<1 512>}, slot {class=ScrollBoxSlot}}
  TextDedicationEpicGames {
   widgetClassPath=&CommonTextBlock
   parent=ScrollBoxCredits
   widget=TextDedicationEpicGames
   properties {
    justification=Center
    style=&TextStyle-Base-M-B_C
    text=Epic Games
    wrapTextAt=941.4548
    font {
     bForceMonospaced=false
     bMaterialIsStencil=false
     fontMaterial=null
     fontObject=&NotoSans
     letterSpacing=0
     monospacedWidth=1
     size=24
     skewAmount=0
     typefaceFontName=NotoSans-Bold
     outlineSettings {
      bApplyOutlineToDropShadows=false
      bMiteredCorners=false
      bSeparateFillAlpha=false
      outlineMaterial=null
      outlineSize=0
      outlineColor=<rgba 0 0 0 1>
     }
    }
    shadowOffset=<0 0>
   }
   slot {class=ScrollBoxSlot, properties {horizontalAlignment=HAlign_Center}}
  }
  Spacer_6 {widgetClassPath=&Spacer, parent=ScrollBoxCredits, widget=Spacer_6, properties {size=<1 1200>}, slot {class=ScrollBoxSlot}}
  ImageMask {
   widgetClassPath=&Image
   bIsVariable=true
   parent=Overlay_1
   widget=ImageMask
   properties {colorAndOpacity=<rgba 0 0 0 1>}
   slot {class=OverlaySlot, properties {horizontalAlignment=HAlign_Fill, verticalAlignment=VAlign_Fill}}
  }
  OverlayCloseButton {
   widgetClassPath=&Overlay
   bIsVariable=true
   parent=Overlay_1
   widget=OverlayCloseButton
   slot {class=OverlaySlot, properties {horizontalAlignment=HAlign_Right, padding=<ltrb 0 32 32 0>}}
  }
  VisibilityBorder_Touch {
   widgetClassPath=&CommonHardwareVisibilityBorder
   parent=OverlayCloseButton
   widget=VisibilityBorder_Touch
   properties {visibilityQuery {
     autoDescription=" ANY( Input.Touch )"
     tokenStreamVersion=0
     userDescription=""
     queryTokenStream [0,1,1,1,0]
     tagDictionary [{tagName=Input.Touch}]
    }}
   slot {class=OverlaySlot}
  }
  Button_CloseTouch {
   widgetClassPath=/FortUILibrary/Foundation/Buttons/CloseButton
   bInherited=true
   bIsVariable=true
   parent=VisibilityBorder_Touch
   widget=Button_CloseTouch
   properties {
    onButtonBaseClicked=(&self:WidgetTree.Button_CloseTouch.HandleOnButtonClicked)
    onButtonBaseDoubleClicked=(&self:WidgetTree.Button_CloseTouch.HandleOnButtonDoubleClicked)
    onButtonBaseHovered=(&self:WidgetTree.Button_CloseTouch.HandleOnButtonHovered)
    onButtonBaseUnhovered=(&self:WidgetTree.Button_CloseTouch.HandleOnButtonUnhovered)
    onSelectedChangedBase=(&self:WidgetTree.Button_CloseTouch.HandleOnSelectedChanged)
   }
   slot {class=BorderSlot, properties {padding=<ltrb 0 0 0 0>}}
  }
 ]
}
